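# Physics asset for the dynamic accessories of the female medium Bullseye Tech body: an armband knot with tails, and a necklace with tags.
# Spine capsules form the torso; the armband master hangs off the upper spine with free rotation, branching through the knot into high and low two-segment tails.
# The necklace box also attaches to the upper spine and carries left and right tag boxes.
# Accessory joints below the master use limited swing and tight twist.
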PhysicsAsset /BRCosmetics/Characters/Player/Female/Medium/Bodies/F_MED_Bullseye_Tech/Meshes/F_MED_Bullseye_Tech_Dyn
bodies {
 dyn_armband_hi_01 [{shapeType=Capsule, length=2, radius=0.716, center=<0.8 0 0>, rotation=<rot 90 0 0>}]
 dyn_armband_hi_02 [{shapeType=Capsule, length=2, radius=0.716, center=<0.8 0 0>, rotation=<rot 90 0 0>}]
 dyn_armband_knot [{shapeType=Capsule, length=2, radius=0.716, center=<0.05 0 0>, rotation=<rot 90 0 0>}]
 dyn_armband_knot_MSTR [{shapeType=Capsule, length=0.5, radius=0.5, rotation=<rot 90 0 0>}]
 dyn_armband_low_01 [{shapeType=Capsule, length=2, radius=0.716, center=<0.8 0 0>, rotation=<rot 90 0 0>}]
 dyn_armband_low_02 [{shapeType=Capsule, length=2, radius=0.716, center=<0.8 0 0>, rotation=<rot 90 0 0>}]
 dyn_necklace [{shapeType=Box, extentX=9.9375, extentY=14.6875, extentZ=1.0625, center=<5.5882 0 -0.0002>}]
 dyn_tags_l [{shapeType=Box, extentX=2.9375, extentY=3.3125, extentZ=2.4375, center=<1.2168 0 -0.0003>}]
 dyn_tags_r [{shapeType=Box, extentX=2.9375, extentY=3.3125, extentZ=2.4375, center=<1.2168 0 -0.0003>}]
 spine_03 [{shapeType=Capsule, length=13.0625, radius=8.5312, center=<1.7806 2.4537 0>}]
 spine_04 [{shapeType=Capsule, length=13.0625, radius=8.5312, center=<1.1429 4.8584 0>}]
 spine_05 [{shapeType=Capsule, length=13.0625, radius=8.5312, center=<0.0053 0.4764 0>}]
}
constraints [
 {
  bone1Name=dyn_armband_knot_MSTR
  bone2Name=spine_05
  swing1LimitDegrees=45
  swing1Motion=Free
  swing2LimitDegrees=45
  swing2Motion=Free
  twistLimitDegrees=45
  twistMotion=Free
 }
 {bone1Name=dyn_armband_knot, bone2Name=dyn_armband_knot_MSTR, swing1LimitDegrees=45, swing2LimitDegrees=45, twistLimitDegrees=2}
 {bone1Name=dyn_armband_low_01, bone2Name=dyn_armband_knot, swing1LimitDegrees=35, swing2LimitDegrees=35, twistLimitDegrees=2}
 {bone1Name=dyn_armband_low_02, bone2Name=dyn_armband_low_01, swing1LimitDegrees=35, swing2LimitDegrees=35, twistLimitDegrees=2}
 {bone1Name=dyn_armband_hi_01, bone2Name=dyn_armband_knot, swing1LimitDegrees=35, swing2LimitDegrees=35, twistLimitDegrees=2}
 {bone1Name=dyn_armband_hi_02, bone2Name=dyn_armband_hi_01, swing1LimitDegrees=35, swing2LimitDegrees=35, twistLimitDegrees=2}
 {bone1Name=dyn_tags_l, bone2Name=dyn_necklace, swing1LimitDegrees=45, swing2LimitDegrees=45, twistLimitDegrees=5}
 {bone1Name=dyn_tags_r, bone2Name=dyn_necklace, swing1LimitDegrees=45, swing2LimitDegrees=45, twistLimitDegrees=5}
 {bone1Name=dyn_necklace, bone2Name=spine_05, swing1LimitDegrees=45, swing2LimitDegrees=45, twistLimitDegrees=5}
 {
  bone1Name=spine_05
  bone2Name=spine_04
  swing1LimitDegrees=45
  swing1Motion=Free
  swing2LimitDegrees=45
  swing2Motion=Free
  twistLimitDegrees=45
  twistMotion=Free
 }
 {
  bone1Name=spine_04
  bone2Name=spine_03
  swing1LimitDegrees=45
  swing1Motion=Free
  swing2LimitDegrees=45
  swing2Motion=Free
  twistLimitDegrees=45
  twistMotion=Free
 }
]
