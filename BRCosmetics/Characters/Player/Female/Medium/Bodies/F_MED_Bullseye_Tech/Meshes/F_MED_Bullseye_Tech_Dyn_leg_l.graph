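# Physics asset for dynamic shin straps on the left leg of the medium female Bullseye Tech cosmetic body.
# The top and bottom straps each have an offset capsule body constrained to a smaller capsule on their MSTR bone.
# Both constraints lock twist and the first swing axis, leaving only the second swing axis unlocked, so each strap hinges on a single axis.
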PhysicsAsset /BRCosmetics/Characters/Player/Female/Medium/Bodies/F_MED_Bullseye_Tech/Meshes/F_MED_Bullseye_Tech_Dyn_leg_l
bodies {
 dyn_shin_strap_bott [{shapeType=Capsule, length=2, radius=1, center=<1.5 0 0>, rotation=<rot 90 0 0>}]
 dyn_shin_strap_bott_MSTR [{shapeType=Capsule, length=0.5, radius=0.5, rotation=<rot 90 0 0>}]
 dyn_shin_strap_top [{shapeType=Capsule, length=2, radius=1, center=<1.5 0 0>, rotation=<rot 90 0 0>}]
 dyn_shin_strap_top_MSTR [{shapeType=Capsule, length=0.5, radius=0.5, rotation=<rot 90 0 0>}]
}
constraints [
 {
  bone1Name=dyn_shin_strap_top
  bone2Name=dyn_shin_strap_top_MSTR
  swing1LimitDegrees=1
  swing1Motion=Locked
  swing2LimitDegrees=5
  twistLimitDegrees=45
  twistMotion=Locked
 }
 {
  bone1Name=dyn_shin_strap_bott
  bone2Name=dyn_shin_strap_bott_MSTR
  swing1LimitDegrees=1
  swing1Motion=Locked
  swing2LimitDegrees=5
  twistLimitDegrees=45
  twistMotion=Locked
 }
]
