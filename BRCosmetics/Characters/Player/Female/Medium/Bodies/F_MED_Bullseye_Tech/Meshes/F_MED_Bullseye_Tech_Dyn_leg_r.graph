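# Physics asset for the right-leg shoelace dynamics on the F_MED_Bullseye_Tech female player cosmetic body.
# A large calf_r capsule supports a small master shoelace body, whose constraint to the calf leaves swing and twist free.
# Inner and outer laces each hang from that master as two-capsule chains, with narrow swing and tighter twist limit angles on every link.
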PhysicsAsset /BRCosmetics/Characters/Player/Female/Medium/Bodies/F_MED_Bullseye_Tech/Meshes/F_MED_Bullseye_Tech_Dyn_leg_r
bodies {
 calf_r [{shapeType=Capsule, length=16.4688, radius=5.125, center=<31.3671 4.7276 0>, rotation=<rot 90 0 -5.4825>}]
 dyn_shoelace_in_r_01 [{shapeType=Capsule, length=1.5, radius=0.5, center=<1 0 0>, rotation=<rot 90 0 0>}]
 dyn_shoelace_in_r_02 [{shapeType=Capsule, length=1.5, radius=0.5, center=<1 0 0>, rotation=<rot 90 0 0>}]
 dyn_shoelace_out_r_01 [{shapeType=Capsule, length=1.5, radius=0.5, center=<1 0 0>, rotation=<rot 90 0 0>}]
 dyn_shoelace_out_r_02 [{shapeType=Capsule, length=1.5, radius=0.5, center=<1 0 0>, rotation=<rot 90 0 0>}]
 dyn_shoelace_r_MSTR [{shapeType=Capsule, length=0.5, radius=0.5, rotation=<rot 90 0 0>}]
}
constraints [
 {bone1Name=dyn_shoelace_out_r_01, bone2Name=dyn_shoelace_r_MSTR, swing1LimitDegrees=15, swing2LimitDegrees=15, twistLimitDegrees=5}
 {bone1Name=dyn_shoelace_out_r_02, bone2Name=dyn_shoelace_out_r_01, swing1LimitDegrees=15, swing2LimitDegrees=15, twistLimitDegrees=5}
 {bone1Name=dyn_shoelace_in_r_01, bone2Name=dyn_shoelace_r_MSTR, swing1LimitDegrees=15, swing2LimitDegrees=15, twistLimitDegrees=5}
 {bone1Name=dyn_shoelace_in_r_02, bone2Name=dyn_shoelace_in_r_01, swing1LimitDegrees=15, swing2LimitDegrees=15, twistLimitDegrees=5}
 {
  bone1Name=dyn_shoelace_r_MSTR
  bone2Name=calf_r
  swing1LimitDegrees=45
  swing1Motion=Free
  swing2LimitDegrees=45
  swing2Motion=Free
  twistLimitDegrees=45
  twistMotion=Free
 }
]
