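# Physics asset for the earmuff straps on the female medium Bullseye Tech character cosmetic, set up for secondary motion of those straps.
# Each side has a two-capsule strap chain constrained from the head capsule, with matching swing limits and a tighter twist limit at each joint.
# An unconstrained left clavicle capsule adds a body shape near the shoulder, with no right-side counterpart.
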PhysicsAsset /BRCosmetics/Characters/Player/Female/Medium/Bodies/F_MED_Bullseye_Tech/Meshes/Parts/F_MED_Bullseye_Tech_Dyn
bodies {
 clavicle_l [{shapeType=Capsule, length=8.1069, radius=4.9085, center=<0.8141 0.0001 3.9825>, rotation=<rot 81.2029 0 0>}]
 dyn_earmuff_strap_l_01 [{shapeType=Capsule, length=3.3175, radius=0.6925, center=<1.9419 0 0>, rotation=<rot 90 0 0>}]
 dyn_earmuff_strap_l_02 [{shapeType=Capsule, length=3.3175, radius=0.6925, center=<1.9419 0 0>, rotation=<rot 90 0 0>}]
 dyn_earmuff_strap_r_01 [{shapeType=Capsule, length=3.3175, radius=0.6925, center=<1.9419 0 0>, rotation=<rot 90 0 0>}]
 dyn_earmuff_strap_r_02 [{shapeType=Capsule, length=3.3175, radius=0.6925, center=<1.9419 0 0>, rotation=<rot 90 0 0>}]
 head [{shapeType=Capsule, length=3, radius=3, rotation=<rot 90 0 0>}]
}
constraints [
 {bone1Name=dyn_earmuff_strap_l_01, bone2Name=head, swing1LimitDegrees=20, swing2LimitDegrees=20, twistLimitDegrees=5}
 {bone1Name=dyn_earmuff_strap_l_02, bone2Name=dyn_earmuff_strap_l_01, swing1LimitDegrees=20, swing2LimitDegrees=20, twistLimitDegrees=5}
 {bone1Name=dyn_earmuff_strap_r_01, bone2Name=head, swing1LimitDegrees=20, swing2LimitDegrees=20, twistLimitDegrees=5}
 {bone1Name=dyn_earmuff_strap_r_02, bone2Name=dyn_earmuff_strap_r_01, swing1LimitDegrees=20, swing2LimitDegrees=20, twistLimitDegrees=5}
]
