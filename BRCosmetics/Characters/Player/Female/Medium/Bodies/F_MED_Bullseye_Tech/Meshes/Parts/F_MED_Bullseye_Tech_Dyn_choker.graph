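# Physics asset for the front choker collar of the female medium Bullseye Tech cosmetic body, defining its collision bodies and joints.
# A constraint chain links the collar top through a small master capsule to spine_05, whose large capsule covers the upper torso.
# Three surface capsules (top, middle, bottom) attach directly to spine_05.
# The collar top uses compound capsules and angled boxes; all joints leave swing and twist free.
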PhysicsAsset /BRCosmetics/Characters/Player/Female/Medium/Bodies/F_MED_Bullseye_Tech/Meshes/Parts/F_MED_Bullseye_Tech_Dyn_choker
bodies {
 dyn_collar_fr_MSTR [{shapeType=Capsule, length=0.5, radius=0.5}]
 dyn_collar_fr_surface_bott [{shapeType=Capsule, length=2.9375, radius=3.8751, center=<0.0046 -3.9676 0>, rotation=<rot 90 -90 -95.9734>}]
 dyn_collar_fr_surface_mid [{shapeType=Capsule, length=2.9375, radius=3.8751, center=<-0.0003 -3.9353 0>, rotation=<rot 90 0 1.2335>}]
 dyn_collar_fr_surface_top [{shapeType=Capsule, length=2.9375, radius=3.8751, center=<0 -3.9237 0>, rotation=<rot 90 0 0>}]
 dyn_collar_fr_top [
  {shapeType=Capsule, length=5.2954, radius=1.2665, center=<-1.1001 1.0724 0>}
  {shapeType=Capsule, length=5.2954, radius=1.2665, center=<0.7654 1.0724 0>}
  {shapeType=Box, extentX=2.2898, extentY=0.9731, extentZ=7.5, center=<-0.2941 0.4161 0>, rotation=<rot 0 -6.6745 0>}
  {shapeType=Box, extentX=2.1899, extentY=1.0981, extentZ=7.5, center=<-0.3379 -0.6383 2.9095>, rotation=<rot 0 0 -50.0001>}
  {shapeType=Box, extentX=2.1899, extentY=1.0981, extentZ=7.5, center=<-0.3497 -1.1468 -3.5689>, rotation=<rot 0 0 50.0002>}
 ]
 spine_05 [{shapeType=Capsule, length=12.8031, radius=8.6478, center=<1.9974 -2.0754 -0.7093>, rotation=<rot -84.9994 -0.0001 -20.0002>}]
}
constraints [
 {
  bone1Name=dyn_collar_fr_MSTR
  bone2Name=spine_05
  swing1LimitDegrees=45
  swing1Motion=Free
  swing2LimitDegrees=45
  swing2Motion=Free
  twistLimitDegrees=45
  twistMotion=Free
 }
 {
  bone1Name=dyn_collar_fr_top
  bone2Name=dyn_collar_fr_MSTR
  swing1LimitDegrees=45
  swing1Motion=Free
  swing2LimitDegrees=45
  swing2Motion=Free
  twistLimitDegrees=45
  twistMotion=Free
 }
 {
  bone1Name=dyn_collar_fr_surface_top
  bone2Name=spine_05
  swing1LimitDegrees=45
  swing1Motion=Free
  swing2LimitDegrees=45
  swing2Motion=Free
  twistLimitDegrees=45
  twistMotion=Free
 }
 {
  bone1Name=dyn_collar_fr_surface_mid
  bone2Name=spine_05
  swing1LimitDegrees=45
  swing1Motion=Free
  swing2LimitDegrees=45
  swing2Motion=Free
  twistLimitDegrees=45
  twistMotion=Free
 }
 {
  bone1Name=dyn_collar_fr_surface_bott
  bone2Name=spine_05
  swing1LimitDegrees=45
  swing1Motion=Free
  swing2LimitDegrees=45
  swing2Motion=Free
  twistLimitDegrees=45
  twistMotion=Free
 }
]
