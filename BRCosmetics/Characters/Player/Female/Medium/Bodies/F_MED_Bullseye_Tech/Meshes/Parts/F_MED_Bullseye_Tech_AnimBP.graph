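# Post-attach AnimBP for the female Bullseye Tech body part, adding secondary motion to its collar and choker.
# It copies the attached parent mesh's pose, then weighted bone constraints drive three front-collar surface bones from head, neck, clavicle and upper-spine bones.
# Two chained RigidBody nodes then simulate separate physics assets in base-bone space: the main one relative to spine_05, the choker relative to neck_02.
# The event graph holds no logic.
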AnimBlueprint /BRCosmetics/Characters/Player/Female/Medium/Bodies/F_MED_Bullseye_Tech/Meshes/Parts/F_MED_Bullseye_Tech_AnimBP
target_skeleton=&/F_MED_Bullseye_Tech_Skeleton
graphs [
 AnimGraph {nodes [
   n1 = "Constraint - dyn_collar_fr_surface_bott (PPPPPP)"!(
    ComponentPose=Local To Component(LocalPose=Copy Pose From Mesh() {bUseAttachedParent=true})
    Alpha=1
    ConstraintWeights_0=0.208
    ConstraintWeights_1:float=0.012
    ConstraintWeights_2:float=0.367
    ConstraintWeights_3:float=0.004
    ConstraintWeights_4:float=0.235
    ConstraintWeights_5:float=0.174
   ) -> (Pose) {
    alphaBoolBlend {blendOption=Linear}
    boneToModify {boneName=dyn_collar_fr_surface_bott}
    constraintSetup [
     {targetBone {boneName=head}}
     {targetBone {boneName=clavicle_l}}
     {targetBone {boneName=spine_05}}
     {targetBone {boneName=clavicle_r}}
     {targetBone {boneName=neck_02}}
     {targetBone {boneName=neck_01}}
    ]
    constraintWeights [1,1,1,1,1,1]
   }
   n2 = "Constraint - dyn_collar_fr_surface_mid (PPPP)"!(
    ComponentPose=@n1
    Alpha=1
    ConstraintWeights_0=0.319
    ConstraintWeights_1:float=0.323
    ConstraintWeights_2:float=0.222
    ConstraintWeights_3:float=0.136
   ) -> (Pose) {
    alphaBoolBlend {blendOption=Linear}
    boneToModify {boneName=dyn_collar_fr_surface_mid}
    constraintSetup [{targetBone {boneName=neck_02}}, {targetBone {boneName=head}}, {targetBone {boneName=neck_01}}, {targetBone {boneName=spine_05}}]
    constraintWeights [1,1,1,1]
   }
   n3 = "Constraint - dyn_collar_fr_surface_top (PPP)"!(ComponentPose=@n2, Alpha=1, ConstraintWeights_0=0.297, ConstraintWeights_1:float=0.55, ConstraintWeights_2:float=0.153) -> (Pose) {
    alphaBoolBlend {blendOption=Linear}
    boneToModify {boneName=dyn_collar_fr_surface_top}
    constraintSetup [{targetBone {boneName=neck_02}}, {targetBone {boneName=head}}, {targetBone {boneName=neck_01}}]
    constraintWeights [1,1,1]
   }
   n4 = RigidBody(ComponentPose=@n3) {
    bClampLinearTranslationLimitToRefPose=true
    bOverrideWorldGravity=true
    overridePhysicsAsset=&/F_MED_Bullseye_Tech_Dyn
    simulationSpace=BaseBoneSpace
    alphaBoolBlend {blendOption=Linear}
    baseBoneRef {boneName=spine_05}
    componentAppliedLinearAccClamp=<3000 3000 3000>
    componentLinearAccScale=<0.5 0.4 0.4>
    componentLinearVelScale=<0.5 0.4 0.4>
    overrideWorldGravity {z=-980}
   }
   n5 = RigidBody(ComponentPose=@n4) {
    bClampLinearTranslationLimitToRefPose=true
    bOverrideWorldGravity=true
    overridePhysicsAsset=&/F_MED_Bullseye_Tech_Dyn_choker
    simulationSpace=BaseBoneSpace
    alphaBoolBlend {blendOption=Linear}
    baseBoneRef {boneName=neck_02}
    componentAppliedLinearAccClamp=<0 0 0>
   }
   Output Pose(Result=Component To Local(ComponentPose=@n5)) {layerGroup=None}
  ]}
 EventGraph {nodes [Event BlueprintUpdateAnimation()]}
]
variables []
